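; Pawn component for the Figment Golden POI feature that holds references to its owning player pawn and the POI manager.
; One function finds all actors of the V2 Golden POI manager class and, if any exist, caches one from the array.
; The other casts the component's owner to a Fort player pawn and caches it.
; Only these two initialization functions appear, with no calls to them or further gameplay logic.
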(fn InitializeManagerReference
  (bind _outactors (Actor|GetAllActorsOfClass "/Figment_GoldenPOI/Gameplay/B_Figment_GoldenPOI_Manager_V2.B_Figment_GoldenPOI_Manager_V2_C"))
  (if (Utilities|Array|IsNotEmpty _outactors)
    (Variables|Default|SetFigmentGoldenPOIManager (Utilities|Array|Get(acopy) _outactors))))

(fn InitializeOwnerReference
  (bind _asfort_player_pawn (Utilities|Casting|CastToFortPlayerPawn (Components|GetOwner)))
  (Variables|Default|SetOwnerPawn _asfort_player_pawn))
